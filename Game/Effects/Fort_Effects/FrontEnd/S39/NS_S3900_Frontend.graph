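# Niagara system for the S39 frontend that sends occasional mesh birds across the scene.
# Two mirrored emitters spawn birds probabilistically on opposite Y sides and launch them toward each other with randomized height and speed.
# A smoothed flap value and speed-scaled phase counter feed dynamic material parameters, presumably animating wings in the material.
# Three rising, wind-driven sprite smoke emitters, two thin and one wide, are present but disabled.
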NiagaraSystem /Game/Effects/Fort_Effects/FrontEnd/S39/NS_S3900_Frontend
dependencies {
 usedDataInterfaces [
  &NiagaraDataInterfaceCurve
  &NiagaraDataInterfaceColorCurve
  &NiagaraDataInterfaceSpriteRendererInfo
 ]
 usedDynamicInputs [&SimulationPosition, &UniformRangedFloat]
 usedModules [&SystemState]
}
emitters {
 Birds {
  emitterUpdateScript [
   &EmitterState()
   &SpawnRate(Spawn Probability=0.2, SpawnRate=0.7, Use Spawn Probability=true) {module=SpawnRate001}
  ]
  particleSpawnScript [
   &InitializeParticle(
    Color=<rgba 0.7344 0.1839 0.1339 1>
    Lifetime=9
    Mesh Scale Mode=::NewEnumerator3
    Mesh Uniform Scale=0.3
    Sprite Size Mode=::NewEnumerator3
    Uniform Sprite Size=100
    UsePositionOffset=false
    Write Material Random=true
   )
   &ShapeLocation(
    Box Size=&RandomRangeVector(Minimum=<0 0 130>, Maximum=<1 1 200>)
    MaskSpawnGroup=false
    Offset=<0 700 400>
    Offset Mode=::NewEnumerator0
    Override Seed=false
    Shape Primitive=::NewEnumerator2
    Surface Only=false
    UseOwnerScale=true
   )
   &AddVelocity(
    Velocity=&RandomRangeVector(Minimum=<0 -1 -0.1>, Maximum=<0 -1 0.25>)
    Velocity Speed Scale=&RandomRangeFloat(Minimum=190, Maximum=210)
   )
  ]
  particleUpdateScript [
   &ParticleState()
   /SetVariables(Particles.RandomDirection:Vector3f=&Lerp_Vector(
     A=&MakeVector(
      X=&Waveform(
       Global Phase=&Add_Float(A=$Particles.Age, B=&RandomRangeFloat(Minimum=0, Maximum=5))
       "Waveform [1]"=::NewEnumerator8
       "[1] Frequency"=1
       "[1] Amplitude Min/Max"=<-1 1>
       "[2] Phase"=$Emitter.Age
       "[3] Phase"=$Emitter.Age
      )
      Y=&Waveform(
       Global Phase=&Add_Float(A=$Particles.Age, B=&RandomRangeFloat(Minimum=0, Maximum=5))
       "Waveform [1]"=::NewEnumerator8
       "[1] Frequency"=1
       "[1] Amplitude Min/Max"=<-1 1>
       "[2] Phase"=$Emitter.Age
       "[3] Phase"=$Emitter.Age
      )
      Z=0
     )
     B=<1 0 0>
     Alpha=&Waveform(
      Global Phase=&Add_Float(A=$Particles.Age, B=&RandomRangeFloat(Minimum=0, Maximum=5))
      Global Clamp Min=0.4
      "Waveform [1]"=::NewEnumerator8
      "[1] Frequency"=1
      "[1] Amplitude Min/Max"=<0 1>
      "[2] Phase"=$Emitter.Age
      "[3] Phase"=$Emitter.Age
     )
    )) {isSetParameters=true, module=SetVariables_#1}
   &InterpolateOverTime(
    Rate Of Change=0.5
    Target Vector Value:Vector3f=&NormalizeVector(Vector To Normalize=$Particles.RandomDirection, Safe Normalize=true, Safe Fallback Threshold=0, Safe Fallback Vector=<0 0 0.001>)
    Variable Type=::NewEnumerator2
   ) {module=InterpolateOverTime001}
   /SetVariables(Particles.FlapTarget=1) {isSetParameters=true, module=SetVariables_#2}
   &InterpolateOverTime(
    Rate Of Change:NiagaraFloat=&Lerp_Float(A=1, B=1, Alpha=$Particles.FlapTarget)
    Target Float Value:NiagaraFloat=$Particles.FlapTarget
   )
   /SetVariables(Particles.FlapSmoothed:NiagaraFloat=$Particles.InterpolateOverTime.InterpolateOverTime_Float.Moving Average) {isSetParameters=true, module=SetVariables_#3}
   /Niagara/Modules/Update/Utility/IncrementOverTime(Rate of Change:NiagaraFloat=&Multiply_Float(
     A=&Add_Float(A=0.15, B=$Particles.FlapSmoothed)
     B=&RemapRange(
      Input Value To Remap=&VectorLength(Vector=$Particles.Velocity)
      Input Min=0
      Input Max=5000
      Output Min=1
      Output Max=5
     )
    ))
   &SolveForcesAndVelocity(Clamp Velocity=true, Limit Acceleration=false, Speed Limit=800)
   &DynamicMaterialParameters(
    Index 0 Param 1=&Multiply_Float(A=$Particles.FlapSmoothed, B=1)
    Index 0 Param 2=&Multiply_Float(A=$Particles.IncrementOverTime.Counter, B=1)
    Module.Param0WriteEnabled=true
    Module.Param1WriteEnabled=true
    Module.Param2WriteEnabled=false
    Module.Param3WriteEnabled=false
   ) {module=DynamicMaterialParameters001}
  ]
  rendererClasses [&NiagaraMeshRendererProperties]
  renderers [{rendererClass=&NiagaraMeshRendererProperties, rendererIndex=0}]
 }
 Birds001 {
  emitterUpdateScript [
   &EmitterState()
   &SpawnRate(Spawn Probability=0.1, Use Spawn Probability=true) {module=SpawnRate001}
  ]
  particleSpawnScript [
   &InitializeParticle(
    Color=<rgba 0.7344 0.1839 0.1339 1>
    Lifetime=9
    Mesh Scale Mode=::NewEnumerator3
    Mesh Uniform Scale=0.3
    Sprite Size Mode=::NewEnumerator3
    Uniform Sprite Size=100
    UsePositionOffset=false
    Write Material Random=true
   )
   &ShapeLocation(
    Box Size=&RandomRangeVector(Minimum=<0 0 -200>, Maximum=<1 1 200>)
    MaskSpawnGroup=false
    Offset=<0 -700 300>
    Offset Mode=::NewEnumerator0
    Override Seed=false
    Shape Primitive=::NewEnumerator2
    Surface Only=false
    UseOwnerScale=true
   )
   &AddVelocity(
    Velocity=&RandomRangeVector(Minimum=<0 1 0.05>, Maximum=<0 1 0.25>)
    Velocity Speed Scale=&RandomRangeFloat(Minimum=190, Maximum=210)
   )
  ]
  particleUpdateScript [
   &ParticleState()
   /SetVariables(Particles.RandomDirection:Vector3f=&Lerp_Vector(
     A=&MakeVector(
      X=&Waveform(
       Global Phase=&Add_Float(A=$Particles.Age, B=&RandomRangeFloat(Minimum=0, Maximum=5))
       "Waveform [1]"=::NewEnumerator8
       "[1] Frequency"=1
       "[1] Amplitude Min/Max"=<-1 1>
       "[2] Phase"=$Emitter.Age
       "[3] Phase"=$Emitter.Age
      )
      Y=&Waveform(
       Global Phase=&Add_Float(A=$Particles.Age, B=&RandomRangeFloat(Minimum=0, Maximum=5))
       "Waveform [1]"=::NewEnumerator8
       "[1] Frequency"=1
       "[1] Amplitude Min/Max"=<-1 1>
       "[2] Phase"=$Emitter.Age
       "[3] Phase"=$Emitter.Age
      )
      Z=0
     )
     B=<1 0 0>
     Alpha=&Waveform(
      Global Phase=&Add_Float(A=$Particles.Age, B=&RandomRangeFloat(Minimum=0, Maximum=5))
      Global Clamp Min=0.4
      "Waveform [1]"=::NewEnumerator8
      "[1] Frequency"=1
      "[1] Amplitude Min/Max"=<0 1>
      "[2] Phase"=$Emitter.Age
      "[3] Phase"=$Emitter.Age
     )
    )) {isSetParameters=true, module=SetVariables_#1}
   &InterpolateOverTime(
    Rate Of Change=0.5
    Target Vector Value:Vector3f=&NormalizeVector(Vector To Normalize=$Particles.RandomDirection, Safe Normalize=true, Safe Fallback Threshold=0, Safe Fallback Vector=<0 0 0.001>)
    Variable Type=::NewEnumerator2
   ) {module=InterpolateOverTime001}
   /SetVariables(Particles.FlapTarget=1) {isSetParameters=true, module=SetVariables_#2}
   &InterpolateOverTime(
    Rate Of Change:NiagaraFloat=&Lerp_Float(A=1, B=1, Alpha=$Particles.FlapTarget)
    Target Float Value:NiagaraFloat=$Particles.FlapTarget
   )
   /SetVariables(Particles.FlapSmoothed:NiagaraFloat=$Particles.InterpolateOverTime.InterpolateOverTime_Float.Moving Average) {isSetParameters=true, module=SetVariables_#3}
   &IncrementOverTime(Rate of Change:NiagaraFloat=&Multiply_Float(
     A=&Add_Float(A=0.15, B=$Particles.FlapSmoothed)
     B=&RemapRange(
      Input Value To Remap=&VectorLength(Vector=$Particles.Velocity)
      Input Min=0
      Input Max=5000
      Output Min=1
      Output Max=5
     )
    ))
   &SolveForcesAndVelocity(Clamp Velocity=true, Limit Acceleration=false, Speed Limit=800)
   &DynamicMaterialParameters(
    Index 0 Param 1=&Multiply_Float(A=$Particles.FlapSmoothed, B=1)
    Index 0 Param 2=&Multiply_Float(A=$Particles.IncrementOverTime.Counter, B=1)
    Module.Param0WriteEnabled=true
    Module.Param1WriteEnabled=true
    Module.Param2WriteEnabled=false
    Module.Param3WriteEnabled=false
   ) {module=DynamicMaterialParameters001}
  ]
  rendererClasses [&NiagaraMeshRendererProperties]
  renderers [{rendererClass=&NiagaraMeshRendererProperties, rendererIndex=0}]
 }
 Smoke_thin {
  bEnabled=false
  emitterUpdateScript [
   &EmitterState()
   &SpawnRate(SpawnRate=3.5, Use Spawn Probability=false)
  ]
  particleSpawnScript [
   &InitializeParticle(
    Color=<rgba 1 1 1 0.5>
    Lifetime=5
    Position Offset=<500 50 -60>
    Sprite Rotation Mode=::NewEnumerator2
    Sprite Size Mode=::NewEnumerator1
    Uniform Sprite Size Max=40
    Uniform Sprite Size Min=30
    UsePositionOffset=true
    Write Material Random=true
   ) {module=InitializeParticle001}
   &ApplyInitialForces(Apply Rotational Force to Rotational Velocity=false)
   &AddVelocity(Velocity=<0 0 1>, Velocity Speed Scale=&RandomRangeFloat(Minimum=40, Maximum=50))
  ]
  particleUpdateScript [
   &ParticleState()
   &WindForce(Scale=0.3, Turbulence Flow Loop Duration=2)
   &Drag(Drag=0.1, Use Linear Drag=true, Use Rotational Drag=true)
   &SolveForcesAndVelocity(Clamp Velocity=false, Limit Acceleration=false)
   &ScaleSpriteSize(Scale Factor=&Vector2DFromFloat(Value=&FloatFromCurve(
      FloatCurve=DataInterface() {propertyValues {Curve {keys [
          {arriveTangent=1.6147, interpMode=RCIM_Cubic, leaveTangent=1.6147, tangentMode=RCTM_User, time=0, value=0.2}
          {arriveTangent=0.2677, interpMode=RCIM_Cubic, leaveTangent=0.2677, tangentMode=RCTM_User, time=1, value=1}
         ]}}}
      Scale Curve=2
     )))
   &DynamicMaterialParameters(
    Index 0 Param 1=&FloatFromCurve(FloatCurve=DataInterface() {propertyValues {Curve {keys [<at -0.0046 1.0113 RCIM_Cubic>,<at 1.0011 0.002>]}}})
    Module.Param0WriteEnabled=true
    Module.Param1WriteEnabled=false
    Module.Param2WriteEnabled=false
    Module.Param3WriteEnabled=true
   ) {module=DynamicMatParams}
   &ScaleColor(
    Scale Alpha=&FloatFromCurve(FloatCurve=DataInterface() {propertyValues {Curve {keys [
         {arriveTangent=0.4422, interpMode=RCIM_Cubic, leaveTangent=0.4422, tangentMode=RCTM_User, time=0, value=0}
         {arriveTangent=1.0142, interpMode=RCIM_Cubic, leaveTangent=1.0142, time=0.25, value=1}
         {arriveTangent=-1.4783, interpMode=RCIM_Cubic, leaveTangent=-1.4783, tangentMode=RCTM_User, time=0.6624, value=0.6717}
         <at 1.0011 0.002>
        ]}}})
    Scale RGB=&MakeVectorFromLinearColorRGB(Color=&ColorFromCurve(ColorCurve=DataInterface() {propertyValues {
        AlphaCurve {keys [<at 0 1 RCIM_Cubic>]}
        BlueCurve {keys [<at 0.3541 0.0251 RCIM_Cubic>,<at 1 0.0527 RCIM_Cubic>]}
        GreenCurve {keys [<at 0.3541 0.0251 RCIM_Cubic>,<at 1 0.0716 RCIM_Cubic>]}
        RedCurve {keys [<at 0.3541 0.0712 RCIM_Cubic>,<at 1 0.5955 RCIM_Cubic>]}
       }}))
    ScaleA=true
    ScaleRGB=true
   )
   &SubUVAnimation(
    Play Rate=0.2
    Random Start Frame=true
    "Sprite Renderer "=DataInterface() {propertyValues {SpriteRenderer=&self:Smoke_thin_1.NiagaraSpriteRendererProperties_1}}
    SubUV Animation Mode:ENiagara_SubUVLookupModeV2=::NewEnumerator3
    UseEndFrame=false
    UsePlayRateScale=false
    UseStartFrame=false
   ) {enabled=false}
   &SubUVAnimation(
    "Sprite Renderer "=DataInterface() {propertyValues {SpriteRenderer=&self:Smoke_thin_1.NiagaraSpriteRendererProperties_0}}
    UseEndFrame=false
    UseStartFrame=false
   ) {module=SubUVAnimation001}
   &FN_Sprite_Attributes(
    Color=&ColorFromCurve(ColorCurve=DataInterface() {propertyValues {
       AlphaCurve {keys [<at 0 0 RCIM_Cubic>,<at 0.3435 0.4 RCIM_Cubic>,<at 1 0 RCIM_Cubic>]}
       BlueCurve {keys [<at 0 0.0251>, {arriveTangent=0.2128, interpMode=RCIM_Cubic, leaveTangent=0.2128, time=0.3182, value=0.1135}, <at 0.8854 0.2135 RCIM_Cubic>]}
       GreenCurve {keys [<at 0 0.0251>, {arriveTangent=0.46, interpMode=RCIM_Cubic, leaveTangent=0.46, time=0.3182, value=0.1135}, <at 0.8854 0.4324 RCIM_Cubic>]}
       RedCurve {keys [<at 0 0.0712>, {arriveTangent=1.049, interpMode=RCIM_Cubic, leaveTangent=1.049, time=0.3182, value=0.7344}, <at 0.8854 1 RCIM_Cubic>]}
      }})
    Create Color Attribute=true
    Create Position Attribute=true
    Create SpriteSize Attribute=true
    Position:NiagaraPosition=&AddVectorToPosition(Position=$Particles.Position, Vector=<-7 0 0>)
    SpriteSize=&Multiply_Vector2D_ByFloat(Float=1.1, Vector2D=$Particles.SpriteSize)
    Write to Module Namespace=true
   )
  ]
  rendererClasses [&NiagaraSpriteRendererProperties]
  renderers [
   {rendererClass=&NiagaraSpriteRendererProperties}
   {rendererClass=&NiagaraSpriteRendererProperties, rendererIndex=1}
  ]
 }
 Smoke_thin001 {
  bEnabled=false
  emitterUpdateScript [
   &EmitterState()
   &SpawnRate(SpawnRate=3.5, Use Spawn Probability=false)
  ]
  particleSpawnScript [
   &InitializeParticle(
    Color=<rgba 1 1 1 0.5>
    Lifetime=5
    Position Offset=<400 50 -60>
    Sprite Rotation Mode=::NewEnumerator2
    Sprite Size Mode=::NewEnumerator1
    Uniform Sprite Size Max=40
    Uniform Sprite Size Min=30
    UsePositionOffset=true
    Write Material Random=true
   ) {module=InitializeParticle001}
   &ApplyInitialForces(Apply Rotational Force to Rotational Velocity=false)
   &AddVelocity(Velocity=<0 0 1>, Velocity Speed Scale=&RandomRangeFloat(Minimum=40, Maximum=50))
  ]
  particleUpdateScript [
   &ParticleState()
   &WindForce(Scale=0.3, Turbulence Flow Loop Duration=2)
   &Drag(Drag=0.1, Use Linear Drag=true, Use Rotational Drag=true)
   &SolveForcesAndVelocity(Clamp Velocity=false, Limit Acceleration=false)
   &ScaleSpriteSize(Scale Factor=&Vector2DFromFloat(Value=&FloatFromCurve(
      FloatCurve=DataInterface() {propertyValues {Curve {keys [
          {arriveTangent=1.6147, interpMode=RCIM_Cubic, leaveTangent=1.6147, tangentMode=RCTM_User, time=0, value=0.2}
          {arriveTangent=0.2677, interpMode=RCIM_Cubic, leaveTangent=0.2677, tangentMode=RCTM_User, time=1, value=1}
         ]}}}
      Scale Curve=2
     )))
   &DynamicMaterialParameters(
    Index 0 Param 1=&FloatFromCurve(FloatCurve=DataInterface() {propertyValues {Curve {keys [<at -0.0046 1.0113 RCIM_Cubic>,<at 1.0011 0.002>]}}})
    Module.Param0WriteEnabled=true
    Module.Param1WriteEnabled=false
    Module.Param2WriteEnabled=false
    Module.Param3WriteEnabled=true
   ) {module=DynamicMatParams}
   &ScaleColor(
    Scale Alpha=&FloatFromCurve(FloatCurve=DataInterface() {propertyValues {Curve {keys [
         {arriveTangent=0.4422, interpMode=RCIM_Cubic, leaveTangent=0.4422, tangentMode=RCTM_User, time=0, value=0}
         {arriveTangent=1.0142, interpMode=RCIM_Cubic, leaveTangent=1.0142, time=0.25, value=1}
         {arriveTangent=-1.4783, interpMode=RCIM_Cubic, leaveTangent=-1.4783, tangentMode=RCTM_User, time=0.6624, value=0.6717}
         <at 1.0011 0.002>
        ]}}})
    Scale RGB=&MakeVectorFromLinearColorRGB(Color=&ColorFromCurve(ColorCurve=DataInterface() {propertyValues {
        AlphaCurve {keys [<at 0 1 RCIM_Cubic>]}
        BlueCurve {keys [<at 0.3541 0.0251 RCIM_Cubic>,<at 1 0.0527 RCIM_Cubic>]}
        GreenCurve {keys [<at 0.3541 0.0251 RCIM_Cubic>,<at 1 0.0716 RCIM_Cubic>]}
        RedCurve {keys [<at 0.3541 0.0712 RCIM_Cubic>,<at 1 0.5955 RCIM_Cubic>]}
       }}))
    ScaleA=true
    ScaleRGB=true
   )
   &SubUVAnimation(
    Play Rate=0.2
    Random Start Frame=true
    "Sprite Renderer "=DataInterface() {propertyValues {SpriteRenderer=&self:Smoke_thin001_1.NiagaraSpriteRendererProperties_1}}
    SubUV Animation Mode:ENiagara_SubUVLookupModeV2=::NewEnumerator3
    UseEndFrame=false
    UsePlayRateScale=false
    UseStartFrame=false
   ) {enabled=false}
   &SubUVAnimation(
    "Sprite Renderer "=DataInterface() {propertyValues {SpriteRenderer=&self:Smoke_thin001_1.NiagaraSpriteRendererProperties_0}}
    UseEndFrame=false
    UseStartFrame=false
   ) {module=SubUVAnimation001}
   &FN_Sprite_Attributes(
    Color=&ColorFromCurve(ColorCurve=DataInterface() {propertyValues {
       AlphaCurve {keys [<at 0 0 RCIM_Cubic>,<at 0.3435 0.4 RCIM_Cubic>,<at 1 0 RCIM_Cubic>]}
       BlueCurve {keys [<at 0 0.0251>, {arriveTangent=0.2128, interpMode=RCIM_Cubic, leaveTangent=0.2128, time=0.3182, value=0.1135}, <at 0.8854 0.2135 RCIM_Cubic>]}
       GreenCurve {keys [<at 0 0.0251>, {arriveTangent=0.46, interpMode=RCIM_Cubic, leaveTangent=0.46, time=0.3182, value=0.1135}, <at 0.8854 0.4324 RCIM_Cubic>]}
       RedCurve {keys [<at 0 0.0712>, {arriveTangent=1.049, interpMode=RCIM_Cubic, leaveTangent=1.049, time=0.3182, value=0.7344}, <at 0.8854 1 RCIM_Cubic>]}
      }})
    Create Color Attribute=true
    Create Position Attribute=true
    Create SpriteSize Attribute=true
    Position:NiagaraPosition=&AddVectorToPosition(Position=$Particles.Position, Vector=<-7 0 0>)
    SpriteSize=&Multiply_Vector2D_ByFloat(Float=1.1, Vector2D=$Particles.SpriteSize)
    Write to Module Namespace=true
   )
  ]
  rendererClasses [&NiagaraSpriteRendererProperties]
  renderers [
   {rendererClass=&NiagaraSpriteRendererProperties}
   {rendererClass=&NiagaraSpriteRendererProperties, rendererIndex=1}
  ]
 }
 Smoke_wide {
  bEnabled=false
  emitterUpdateScript [
   &EmitterState()
   &SpawnRate(SpawnRate=2, Use Spawn Probability=false)
  ]
  particleSpawnScript [
   &InitializeParticle(
    Color=<rgba 1 1 1 0.5>
    Lifetime=6.5
    Position Offset=<500 50 -70>
    Sprite Rotation Mode=::NewEnumerator2
    Sprite Size Mode=::NewEnumerator1
    Uniform Sprite Size Max=120
    Uniform Sprite Size Min=70
    UsePositionOffset=true
    Write Material Random=true
   ) {module=InitializeParticle001}
   &ApplyInitialForces(Apply Rotational Force to Rotational Velocity=false)
   &AddVelocity(Velocity=<0 0 1>, Velocity Speed Scale=&RandomRangeFloat(Minimum=40, Maximum=50))
  ]
  particleUpdateScript [
   &ParticleState()
   &WindForce(Scale=0.3, Turbulence Flow Loop Duration=2)
   &Drag(Drag=0.1, Use Linear Drag=true, Use Rotational Drag=true)
   &SolveForcesAndVelocity(Clamp Velocity=false, Limit Acceleration=false)
   &ScaleSpriteSize(Scale Factor=&Vector2DFromFloat(Value=&FloatFromCurve(
      FloatCurve=DataInterface() {propertyValues {Curve {keys [
          {arriveTangent=1.6147, interpMode=RCIM_Cubic, leaveTangent=1.6147, tangentMode=RCTM_User, time=0, value=0.1}
          {arriveTangent=0.2677, interpMode=RCIM_Cubic, leaveTangent=0.2677, tangentMode=RCTM_User, time=1, value=1}
         ]}}}
      Scale Curve=2
     )))
   &DynamicMaterialParameters(
    Index 0 Param 1=&FloatFromCurve(FloatCurve=DataInterface() {propertyValues {Curve {keys [<at -0.0046 1.0113 RCIM_Cubic>,<at 1.0011 0.002>]}}})
    Module.Param0WriteEnabled=true
    Module.Param1WriteEnabled=false
    Module.Param2WriteEnabled=false
    Module.Param3WriteEnabled=true
   ) {module=DynamicMatParams}
   &ScaleColor(
    Scale Alpha=&FloatFromCurve(FloatCurve=DataInterface() {propertyValues {Curve {keys [
         {arriveTangent=0.4422, interpMode=RCIM_Cubic, leaveTangent=0.4422, tangentMode=RCTM_User, time=0, value=0}
         {arriveTangent=1.0142, interpMode=RCIM_Cubic, leaveTangent=1.0142, time=0.25, value=1}
         {arriveTangent=-1.4783, interpMode=RCIM_Cubic, leaveTangent=-1.4783, tangentMode=RCTM_User, time=0.6624, value=0.6717}
         <at 1.0011 0.002>
        ]}}})
    Scale RGB=&MakeVectorFromLinearColorRGB(Color=&ColorFromCurve(ColorCurve=DataInterface() {propertyValues {
        AlphaCurve {keys [<at 0 1 RCIM_Cubic>]}
        BlueCurve {keys [<at 0.3541 0.0251 RCIM_Cubic>,<at 1 0.0527 RCIM_Cubic>]}
        GreenCurve {keys [<at 0.3541 0.0251 RCIM_Cubic>,<at 1 0.0716 RCIM_Cubic>]}
        RedCurve {keys [<at 0.3541 0.0712 RCIM_Cubic>,<at 1 0.5955 RCIM_Cubic>]}
       }}))
    ScaleA=true
    ScaleRGB=true
   )
   &SubUVAnimation(
    Play Rate=0.2
    Random Start Frame=true
    "Sprite Renderer "=DataInterface() {propertyValues {SpriteRenderer=&self:Smoke_wide_0.NiagaraSpriteRendererProperties_1}}
    SubUV Animation Mode:ENiagara_SubUVLookupModeV2=::NewEnumerator3
    UseEndFrame=false
    UsePlayRateScale=false
    UseStartFrame=false
   ) {enabled=false}
   &SubUVAnimation(
    "Sprite Renderer "=DataInterface() {propertyValues {SpriteRenderer=&self:Smoke_wide_0.NiagaraSpriteRendererProperties_0}}
    UseEndFrame=false
    UseStartFrame=false
   ) {module=SubUVAnimation001}
   &FN_Sprite_Attributes(
    Color=&ColorFromCurve(ColorCurve=DataInterface() {propertyValues {
       AlphaCurve {keys [<at 0 0 RCIM_Cubic>,<at 0.3435 0.4 RCIM_Cubic>,<at 1 0 RCIM_Cubic>]}
       BlueCurve {keys [<at 0 0.0251>, {arriveTangent=0.2128, interpMode=RCIM_Cubic, leaveTangent=0.2128, time=0.3182, value=0.1135}, <at 0.8854 0.2135 RCIM_Cubic>]}
       GreenCurve {keys [<at 0 0.0251>, {arriveTangent=0.46, interpMode=RCIM_Cubic, leaveTangent=0.46, time=0.3182, value=0.1135}, <at 0.8854 0.4324 RCIM_Cubic>]}
       RedCurve {keys [<at 0 0.0712>, {arriveTangent=1.049, interpMode=RCIM_Cubic, leaveTangent=1.049, time=0.3182, value=0.7344}, <at 0.8854 1 RCIM_Cubic>]}
      }})
    Create Color Attribute=true
    Create Position Attribute=true
    Create SpriteSize Attribute=true
    Position:NiagaraPosition=&AddVectorToPosition(Position=$Particles.Position, Vector=<-7 0 0>)
    SpriteSize=&Multiply_Vector2D_ByFloat(Float=1.1, Vector2D=$Particles.SpriteSize)
    Write to Module Namespace=true
   )
  ]
  rendererClasses [&NiagaraSpriteRendererProperties]
  renderers [
   {rendererClass=&NiagaraSpriteRendererProperties}
   {rendererClass=&NiagaraSpriteRendererProperties, rendererIndex=1}
  ]
 }
}
summary {
 systemName=NS_S3900_Frontend
 emitters [
  {bEnabled=false, emitterName=Smoke_wide, rendererClasses [&NiagaraSpriteRendererProperties]}
  {bEnabled=false, emitterName=Smoke_thin, rendererClasses [&NiagaraSpriteRendererProperties]}
  {bEnabled=false, emitterName=Smoke_thin001, rendererClasses [&NiagaraSpriteRendererProperties]}
  {emitterName=Birds001, rendererClasses [&NiagaraMeshRendererProperties]}
  {emitterName=Birds, rendererClasses [&NiagaraMeshRendererProperties]}
 ]
 userVariables [User.Speed {type=NiagaraFloat, defaultValue=0, description=""}]
}
